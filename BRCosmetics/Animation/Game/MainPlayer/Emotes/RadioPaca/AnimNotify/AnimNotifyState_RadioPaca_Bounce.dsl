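; Anim notify state for the RadioPaca emote that feeds timing for a bounce effect to the animated mesh's materials.
; When the notify begins, it writes the current game time to the Timestamp scalar parameter and the notify window's duration to TransitionLength.
; Only the begin event is implemented, so any motion over the window is left to the materials reading those two values.
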(fn Received_NotifyBegin
  (params (MeshComp :type Skeletal%20Mesh%20Component%20Object%20Reference) (Animation :type Anim%20Sequence%20Base%20Object%20Reference) (TotalDuration :type Float%20%28single-precision%29) (EventReference :type Anim%20Notify%20Event%20Reference%20Structure%20%28by%20ref%29))
  (Rendering|Material|SetScalarParameterValueonMaterials MeshComp "Timestamp" (Utilities|Time|GetGameTimeinSeconds MeshComp))
  (Rendering|Material|SetScalarParameterValueonMaterials MeshComp "TransitionLength" TotalDuration)
  (return false))
