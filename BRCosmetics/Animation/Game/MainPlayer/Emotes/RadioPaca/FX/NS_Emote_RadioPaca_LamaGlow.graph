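# Niagara effect for the RadioPaca player emote that draws soft green glow sprites attached to mesh sockets.
# Three instantaneous bursts in one emitter use separate spawn groups, and each group number picks the socket its particles follow through a system socket reader named NewsVan.
# Sprites are nudged by a camera offset and fade out over their lifetime, with color from a curve and alpha scaled by time of day, dimmest at night.
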NiagaraSystem /BRCosmetics/Animation/Game/MainPlayer/Emotes/RadioPaca/FX/NS_Emote_RadioPaca_LamaGlow
dependencies {
 usedDataInterfaces [&NiagaraDataInterfaceColorCurve, &NiagaraDataInterfaceCurve, &NiagaraDataInterfaceSocketReader]
 usedDynamicInputs []
 usedModules [
  &SystemState
  &self:SystemSpawnScript.SystemScriptSource.SystemScriptGraph.NiagaraNodeAssignment_0.SetVariables_#1
 ]
}
emitters {Minimal {
  emitterUpdateScript [
   &EmitterState()
   &SpawnBurst_Instantaneous(Use Spawn Probability=false, UseLoopCountLimit=false)
   &SpawnBurst_Instantaneous(Spawn Group=1, Use Spawn Probability=false, UseLoopCountLimit=false) {module=SpawnBurst_Instantaneous002}
   &SpawnBurst_Instantaneous(Spawn Group=2, Use Spawn Probability=false, UseLoopCountLimit=false) {module=SpawnBurst_Instantaneous001}
  ]
  particleSpawnScript [
   &InitializeParticle(
    Color Mode=::NewEnumerator0
    Lifetime=6
    Sprite Size=<30 30>
    Sprite Size Mode=::NewEnumerator4
    UsePositionOffset=false
    Write Material Random=true
   )
   &CameraOffset(Camera Offset Amount=5)
   &SpriteFacingAndAlignment(Sprite Alignment=<0 0 1>, Sprite Facing=<0 1 0>, WriteAlignment=true, WriteFacing=true) {enabled=false}
  ]
  particleUpdateScript [
   &ParticleState()
   &SocketLocation(
    SamplePosOffset=true
    Socket Index=&SpawnGroup()
    Socket Reader=$System.NewsVan Socket Reader
    Socket Sampling Mode=::NewEnumerator3
    VelocitySpeedLimit=false
   )
   &ScaleColor(
    Scale Alpha=&FloatFromCurve(
     FloatCurve=DataInterface() {propertyValues {Curve {keys [<at 0.25 1>,<at 1 0>]}}}
     Scale Curve=&FNTimeOfDayScalar(Night Scale=0.1, Morning Scale=0.4, Day Scale=0.4, Evening Scale=0.2)
    )
    Scale RGB=&MakeVectorFromLinearColorRGB(Color=&ColorFromCurve(ColorCurve=DataInterface() {propertyValues {
        AlphaCurve {keys [<at 0 0.1 RCIM_Cubic>,<at 0.7107 0>]}
        BlueCurve {keys [<at 0 0 RCIM_Cubic>,<at 1 0.2013 RCIM_Cubic>]}
        GreenCurve {keys [<at 0 0.5 RCIM_Cubic>,<at 1 1 RCIM_Cubic>]}
        RedCurve {keys [<at 0 0.1444 RCIM_Cubic>,<at 1 0.1679 RCIM_Cubic>]}
       }}))
    ScaleA=true
    ScaleRGB=true
   )
  ]
  rendererClasses [&NiagaraSpriteRendererProperties]
  renderers [{rendererClass=&NiagaraSpriteRendererProperties}]
 }}
summary {
 systemName=NS_Emote_RadioPaca_LamaGlow
 emitters [{emitterName=Minimal, rendererClasses [&NiagaraSpriteRendererProperties]}]
 userVariables [
  User.bIsFrontEnd {type=NiagaraBool}
  User.bIsFrontEndPreview {type=NiagaraBool}
  User.OwnerLocalVelocity {type=Vector3f, description="", defaultValue {struct=Vector3f, value=<0 0 0>}}
 ]
}
