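# Niagara side laser burst effect for the RadioPaca player emote.
# Short-lived stretched mesh rays and thin rays fire in bursts, layered with two glow sprites and spinning, drag-slowed sparkle particles; two duplicate emitters are disabled.
# User.VariantSwitch blends between color sets, while bIsFrontEndPreview shrinks the effect and time-of-day scalars dim rays and glows.
# Glows compute light attributes but have only sprite renderers, and most user parameters go unreferenced by these emitters.
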NiagaraSystem /BRCosmetics/Animation/Game/MainPlayer/Emotes/RadioPaca/FX/NS_Emote_RadioPaca_Laser_Side
dependencies {
 usedDataInterfaces [&NiagaraDataInterfaceCurve, &NiagaraDataInterfacePlatformSet]
 usedDynamicInputs [
  &NormalizeVector
  &Divide_Float
  &Add_Integer
  /Niagara/DynamicInputs/Divide/Divide_Int
  &ModuloInt
 ]
 usedModules [
  &SystemState
  &self:SystemSpawnScript.SystemScriptSource.SystemScriptGraph.NiagaraNodeAssignment_0.SetVariables_#1
  &self:SystemSpawnScript.SystemScriptSource.SystemScriptGraph.NiagaraNodeAssignment_1.SetVariables_#2
  &self:SystemSpawnScript.SystemScriptSource.SystemScriptGraph.NiagaraNodeAssignment_2.SetVariables_#3
 ]
}
emitters {
 Particle {
  bEnabled=false
  emitterUpdateScript [
   &EmitterState(
    Inactive Response=::NewEnumerator1
    Life Cycle Mode=::NewEnumerator1
    Loop Behavior=::NewEnumerator1
    Loop Delay=0.02
    Loop Duration=1.5
    UseLoopDelay=true
   )
   &SpawnBurst_Instantaneous(
    Spawn Count=&RandomRangeInt(Minimum=10, Maximum=20)
    Use Spawn Probability=false
    UseLoopCountLimit=false
   )
  ]
  particleSpawnScript [
   &InitializeParticle(
    Color=&LerpLinearColors(
     StartColor=&MakeCustomLinearColorFromBool(Bool=&RandomBool(), True Color=<rgba 0 0.07 0.6 1.5>, False Color=<rgba 0 0.7497 1.5 1>)
     EndColor=&MakeCustomLinearColorFromBool(Bool=&RandomBool(), True Color=<rgba 0.1106 0.1979 0.0158 1.5>, False Color=<rgba 0.5339 1 0.0677 1>)
     LerpFactor=$User.VariantSwitch
    )
    Lifetime Max=0.5
    Lifetime Min=0.3
    Lifetime Mode=::NewEnumerator1
    Position Offset=<0 0 3>
    Sprite Size Mode=::NewEnumerator1
    Uniform Sprite Size Max=8
    UsePositionOffset=true
    Write Material Random=true
   )
   &ShapeLocation(MaskSpawnGroup=false, Override Seed=false, Sphere Radius=2, UseOwnerScale=true) {enabled=false}
   &AddVelocity(
    Module.Use Falloff Distance=false
    Override Seed=false
    Velocity Mode=::NewEnumerator1
    Velocity Speed=&Multiply_Float(
     A=&RandomRangeFloat(Minimum=100, Maximum=500)
     B=&MakeCustomFloatFromBool(Bool=$User.bIsFrontEndPreview, True Float=0.6, False Float=1)
    )
   )
   &SystemLocation(MaskSpawnGroup=false, Offset=<0 -30 -7>)
   &ShapeLocation(Box Size=<33 12 0>, MaskSpawnGroup=false, Override Seed=false, Shape Primitive=::NewEnumerator2, Surface Only=false, UseOwnerScale=true) {module=ShapeLocation001}
  ]
  particleUpdateScript [
   &ParticleState()
   &Drag(Drag=&RandomRangeFloat(Minimum=7, Maximum=5), Use Linear Drag=true, Use Rotational Drag=false)
   &ScaleSpriteSize(Scale Factor=&Vector2DFromFloat(Value=&FloatFromCurve(
      FloatCurve=DataInterface() {propertyValues {Curve {keys [<at 0 0.25>,<at 0.1459 0.9717>,<at 0.8712 0.4109>,<at 1 0>]}}}
      Scale Curve=&MakeCustomFloatFromBool(Bool=$User.bIsFrontEndPreview, True Float=0.7, False Float=1)
     )))
   &ScaleColor(
    Color Value To Scale=$Particles.Color
    Scale Alpha=&FloatFromCurve(FloatCurve=DataInterface() {propertyValues {Curve {keys [
         {arriveTangent=0.103, interpMode=RCIM_Cubic, leaveTangent=0.103, tangentMode=RCTM_User, time=0, value=0}
         {arriveTangent=-3.5179, interpMode=RCIM_Cubic, leaveTangent=-3.5179, tangentMode=RCTM_User, time=0.25, value=1}
         <at 1 0>
        ]}}})
    Scale RGB=<2 2 2>
    ScaleA=true
    ScaleRGB=true
   ) {module=ScaleColor001}
   &SolveForcesAndVelocity(Clamp Velocity=false, Limit Acceleration=false)
   &ScaleSpriteSizeBySpeed(Max Scale Factor=<0.6 2>, Min Scale Factor=<0.4 1>)
   &SpriteRotationRate(Rotation Rate=&RandomRangeFloat(Minimum=-300, Maximum=300))
   &FN_Sprite_Attributes(
    Color=&MultiplyLinearColor(Linear Color=$Particles.Color, Scale Factor=&MakeVector4_FromFloat(Float=0.5))
    Create Color Attribute=true
    Create SpriteSize Attribute=true
    SpriteSize=&Multiply_Vector2D_ByFloat(Float=2, Vector2D=$Particles.SpriteSize)
    Write to Module Namespace=true
   )
  ]
  rendererClasses [&NiagaraSpriteRendererProperties]
  renderers [
   {rendererClass=&NiagaraSpriteRendererProperties}
   {rendererClass=&NiagaraSpriteRendererProperties, rendererIndex=1}
  ]
 }
 Particle001 {
  emitterUpdateScript [
   &EmitterState(Inactive Response=::NewEnumerator1, Life Cycle Mode=::NewEnumerator1, Loop Behavior=::NewEnumerator1, Loop Duration=1.5, UseLoopDelay=false)
   &SpawnBurst_Instantaneous(
    Spawn Count=&RandomRangeInt(Minimum=10, Maximum=20)
    Use Spawn Probability=false
    UseLoopCountLimit=false
   )
  ]
  particleSpawnScript [
   &InitializeParticle(
    Color=&LerpLinearColors(
     StartColor=&MakeCustomLinearColorFromBool(Bool=&RandomBool(), True Color=<rgba 0 0.07 0.6 1.5>, False Color=<rgba 0 0.7497 1.5 1>)
     EndColor=&MakeCustomLinearColorFromBool(Bool=&RandomBool(), True Color=<rgba 0.1106 0.1979 0.0158 1.5>, False Color=<rgba 0.5339 1 0.0677 1>)
     LerpFactor=$User.VariantSwitch
    )
    Lifetime Max=0.4
    Lifetime Min=0.3
    Lifetime Mode=::NewEnumerator1
    Sprite Size Mode=::NewEnumerator1
    Uniform Sprite Size Max=6
    Uniform Sprite Size Min=3
    UsePositionOffset=false
    Write Material Random=true
   )
   &ShapeLocation(MaskSpawnGroup=false, Override Seed=false, Sphere Radius=20, UseOwnerScale=true)
   &AddVelocity(
    Module.Use Falloff Distance=false
    Override Seed=false
    Velocity Mode=::NewEnumerator1
    Velocity Speed=&Multiply_Float(
     A=&RandomRangeFloat(Minimum=100, Maximum=500)
     B=&MakeCustomFloatFromBool(Bool=$User.bIsFrontEndPreview, True Float=0.6, False Float=1)
    )
   )
  ]
  particleUpdateScript [
   &ParticleState()
   &Drag(Drag=&RandomRangeFloat(Minimum=3, Maximum=5), Use Linear Drag=true, Use Rotational Drag=false)
   &ScaleSpriteSize(Scale Factor=&Vector2DFromFloat(Value=&FloatFromCurve(
      FloatCurve=DataInterface() {propertyValues {Curve {keys [<at 0 0.25>,<at 0.2 1>,<at 1 0>]}}}
      Scale Curve=&MakeCustomFloatFromBool(Bool=$User.bIsFrontEndPreview, True Float=0.7, False Float=1)
     )))
   &ScaleColor(
    Color Value To Scale=$Particles.Color
    Scale Alpha=&FloatFromCurve(FloatCurve=DataInterface() {propertyValues {Curve {keys [
         {arriveTangent=0.103, interpMode=RCIM_Cubic, leaveTangent=0.103, tangentMode=RCTM_User, time=0, value=0}
         {arriveTangent=-3.5179, interpMode=RCIM_Cubic, leaveTangent=-3.5179, tangentMode=RCTM_User, time=0.25, value=1}
         <at 1 0>
        ]}}})
    Scale RGB=<2 2 2>
    ScaleA=true
    ScaleRGB=true
   ) {module=ScaleColor001}
   &SolveForcesAndVelocity(Clamp Velocity=false, Limit Acceleration=false)
   &SpriteRotationRate(Rotation Rate=&RandomRangeFloat(Minimum=-300, Maximum=300))
   &FN_Sprite_Attributes(
    Color=&MultiplyLinearColor(Linear Color=$Particles.Color, Scale Factor=&MakeVector4_FromFloat(Float=0.5))
    Create Color Attribute=true
    Create SpriteSize Attribute=true
    SpriteSize=&Multiply_Vector2D_ByFloat(Float=2, Vector2D=$Particles.SpriteSize)
    Write to Module Namespace=true
   )
  ]
  rendererClasses [&NiagaraSpriteRendererProperties]
  renderers [
   {rendererClass=&NiagaraSpriteRendererProperties}
   {rendererClass=&NiagaraSpriteRendererProperties, rendererIndex=1}
  ]
 }
 Rays {
  emitterUpdateScript [
   &EmitterState(Inactive Response=::NewEnumerator1, Life Cycle Mode=::NewEnumerator1, Loop Duration=0.375, UseLoopDelay=false)
   &SpawnRate(SpawnRate=20, Use Spawn Probability=false) {enabled=false}
   &SpawnBurst_Instantaneous(Spawn Count=12, Use Spawn Probability=false)
  ]
  particleSpawnScript [
   &InitializeParticle(
    Color=&LerpLinearColors(
     StartColor=&MakeCustomLinearColorFromBool(Bool=&RandomBool(), True Color=<rgba 0.0011 0.794 4 1.6>, False Color=<rgba 0.4881 0.0048 2 2>)
     EndColor=&MakeCustomLinearColorFromBool(Bool=&RandomBool(), True Color=<rgba 0 0.7465 1.5 1.5>, False Color=<rgba 0 0.238 1 1.5>)
     LerpFactor=$User.VariantSwitch
    )
    Lifetime Max=0.25
    Lifetime Min=0.15
    Lifetime Mode=::NewEnumerator1
    Mesh Scale Max=<0.7 0 0.4>
    Mesh Scale Min=<0.1 1 0.15>
    Mesh Scale Mode=::NewEnumerator2
    UsePositionOffset=false
    Write Material Random=true
   ) {module=InitializeParticle001}
   &CameraOffset(Camera Offset Amount=30)
   &DynamicMaterialParameters(
    Index 0 Param 1=&UniformRangedFloat(Minimum=-1, Maximum=1)
    Index 0 Param 2=1
    Module.Param0WriteEnabled=true
    Module.Param1WriteEnabled=true
    Module.Param2WriteEnabled=true
    Module.Param3WriteEnabled=true
   )
   &InitialMeshOrientation(
    Orientation Vector=<1 0 1>
    Rotation=&Add_Vector(A=<0 0 0>, B=&RandomRangeVector(Minimum=<0.25 0 0.25>, Maximum=<0 0 -0.18>))
    Use Orientation Vector=true
    Use Rotation Vector=true
   ) {module=InitialMeshOrientation001}
  ]
  particleUpdateScript [
   &ParticleState()
   &SolveForcesAndVelocity(Clamp Velocity=false, Limit Acceleration=false)
   &ScaleColor(
    Scale Alpha=&Multiply_Float(
     A=&FloatFromCurve(FloatCurve=DataInterface() {propertyValues {Curve {keys [
          {arriveTangent=9.1628, interpMode=RCIM_Cubic, leaveTangent=9.1628, tangentMode=RCTM_User, time=0, value=0.1998}
          {arriveTangent=-1.3283, interpMode=RCIM_Cubic, leaveTangent=-1.3283, tangentMode=RCTM_User, time=0.15, value=0.5}
          {tangentMode=RCTM_User, time=1, value=0.1998}
         ]}}})
     B=&RandomRangeFloat(Minimum=1, Maximum=1.5)
    )
    Scale RGB=&Multiply_VectorByFloat(Vector=<1 1 1>, Float=&FNTimeOfDayScalar(Night Scale=0.5, Morning Scale=0.3, Evening Scale=0.4))
    ScaleA=true
    ScaleRGB=true
   )
   &ScaleMeshSize(Scale Factor=&MakeVector(
     X=&FloatFromCurve(FloatCurve=DataInterface() {propertyValues {Curve {keys [<at 0 0.8>,<at 0.5 1.2>]}}})
     Y=1
     Z=&FloatFromCurve(FloatCurve=DataInterface() {propertyValues {Curve {keys [<at 0 0.5>,<at 0.5 1>]}}})
    ))
   &ScaleMeshSize(Scale Factor=&VectorFromFloat(Value=&MakeCustomFloatFromBool(Bool=$User.bIsFrontEndPreview, True Float=0.5, False Float=0.9))) {module=ScaleMeshSize001}
  ]
  rendererClasses [&NiagaraMeshRendererProperties]
  renderers [{rendererClass=&NiagaraMeshRendererProperties, rendererIndex=0}]
 }
 Rays001 {
  bEnabled=false
  emitterUpdateScript [
   &EmitterState(Inactive Response=::NewEnumerator1, Life Cycle Mode=::NewEnumerator1, Loop Duration=0.375, UseLoopDelay=false)
   &SpawnRate(SpawnRate=20, Use Spawn Probability=false) {enabled=false}
   &SpawnBurst_Instantaneous(Spawn Count=12, Use Spawn Probability=false)
  ]
  particleSpawnScript [
   &InitializeParticle(
    Color=&LerpLinearColors(
     StartColor=&MakeCustomLinearColorFromBool(Bool=&RandomBool(), True Color=<rgba 0.0011 0.794 4 1.6>, False Color=<rgba 0.4881 0.0048 2 2>)
     EndColor=&MakeCustomLinearColorFromBool(Bool=&RandomBool(), True Color=<rgba 0 0.7465 1.5 1.5>, False Color=<rgba 0 0.238 1 1.5>)
     LerpFactor=$User.VariantSwitch
    )
    Lifetime Max=0.25
    Lifetime Min=0.15
    Lifetime Mode=::NewEnumerator1
    Mesh Scale Max=<0.7 0 0.4>
    Mesh Scale Min=<0.1 1 0.15>
    Mesh Scale Mode=::NewEnumerator2
    UsePositionOffset=false
    Write Material Random=true
   ) {module=InitializeParticle001}
   &CameraOffset(Camera Offset Amount=30)
   &DynamicMaterialParameters(
    Index 0 Param 1=&UniformRangedFloat(Minimum=-1, Maximum=1)
    Index 0 Param 2=1
    Module.Param0WriteEnabled=true
    Module.Param1WriteEnabled=true
    Module.Param2WriteEnabled=true
    Module.Param3WriteEnabled=true
   )
   &InitialMeshOrientation(
    Orientation Vector=<1 0 1>
    Rotation=&Add_Vector(A=<0 0 0>, B=&RandomRangeVector(Minimum=<0.25 0 0.25>, Maximum=<0 0 -0.18>))
    Use Orientation Vector=true
    Use Rotation Vector=true
   ) {module=InitialMeshOrientation001}
  ]
  particleUpdateScript [
   &ParticleState()
   &SolveForcesAndVelocity(Clamp Velocity=false, Limit Acceleration=false)
   &ScaleColor(
    Scale Alpha=&Multiply_Float(
     A=&FloatFromCurve(FloatCurve=DataInterface() {propertyValues {Curve {keys [
          {arriveTangent=9.1628, interpMode=RCIM_Cubic, leaveTangent=9.1628, tangentMode=RCTM_User, time=0, value=0}
          {arriveTangent=-1.3283, interpMode=RCIM_Cubic, leaveTangent=-1.3283, tangentMode=RCTM_User, time=0.15, value=0.5}
          {tangentMode=RCTM_User, time=1, value=0}
         ]}}})
     B=&RandomRangeFloat(Minimum=1, Maximum=1.5)
    )
    Scale RGB=&Multiply_VectorByFloat(Vector=<1 1 1>, Float=&FNTimeOfDayScalar(Night Scale=0.5, Morning Scale=0.3, Evening Scale=0.4))
    ScaleA=true
    ScaleRGB=true
   )
   &ScaleMeshSize(Scale Factor=&MakeVector(
     X=&FloatFromCurve(FloatCurve=DataInterface() {propertyValues {Curve {keys [<at 0 0.8>,<at 0.5 1.2>]}}})
     Y=1
     Z=&FloatFromCurve(FloatCurve=DataInterface() {propertyValues {Curve {keys [<at 0 0.5>,<at 0.5 1>]}}})
    ))
   &ScaleMeshSize(Scale Factor=&VectorFromFloat(Value=&MakeCustomFloatFromBool(Bool=$User.bIsFrontEndPreview, True Float=0.5, False Float=0.9))) {module=ScaleMeshSize001}
  ]
  rendererClasses [&NiagaraMeshRendererProperties]
  renderers [{rendererClass=&NiagaraMeshRendererProperties, rendererIndex=0}]
 }
 Thin_Rays {
  emitterUpdateScript [
   &EmitterState(Inactive Response=::NewEnumerator1, Life Cycle Mode=::NewEnumerator1, Loop Duration=0.375, UseLoopDelay=false)
   &SpawnRate(SpawnRate=20, Use Spawn Probability=false) {enabled=false}
   &SpawnBurst_Instantaneous(Spawn Count=7, Use Spawn Probability=false)
  ]
  particleSpawnScript [
   &InitializeParticle(
    Color=&LerpLinearColors(
     StartColor=&MakeCustomLinearColorFromBool(Bool=&RandomBool(), True Color=<rgba 0.8188 1.9146 2 0.9322>, False Color=<rgba 1.0188 1.6603 2 0.6452>)
     EndColor=&MakeCustomLinearColorFromBool(Bool=&RandomBool(), True Color=<rgba 0 0.7465 1.5 1.5>, False Color=<rgba 0 0.238 1 1.5>)
     LerpFactor=$User.VariantSwitch
    )
    Lifetime Max=0.25
    Lifetime Min=0.15
    Lifetime Mode=::NewEnumerator1
    Mesh Scale Max=<0.5 0 0.02>
    Mesh Scale Min=<0.4 1 0.02>
    Mesh Scale Mode=::NewEnumerator2
    UsePositionOffset=false
    Write Material Random=true
   ) {module=InitializeParticle001}
   &CameraOffset(Camera Offset Amount=30)
   &DynamicMaterialParameters(
    Index 0 Param 1=&UniformRangedFloat(Minimum=-1, Maximum=1)
    Index 0 Param 2=1
    Module.Param0WriteEnabled=true
    Module.Param1WriteEnabled=true
    Module.Param2WriteEnabled=true
    Module.Param3WriteEnabled=true
   )
   &InitialMeshOrientation(
    Orientation Vector=<1 0 1>
    Rotation=&Add_Vector(A=<0 0 0>, B=&RandomRangeVector(Minimum=<0.25 0 0.25>, Maximum=<0 0 -0.18>))
    Use Orientation Vector=true
    Use Rotation Vector=true
   ) {module=InitialMeshOrientation001}
  ]
  particleUpdateScript [
   &ParticleState()
   &SolveForcesAndVelocity(Clamp Velocity=false, Limit Acceleration=false)
   &ScaleColor(
    Scale Alpha=&Multiply_Float(
     A=&FloatFromCurve(FloatCurve=DataInterface() {propertyValues {Curve {keys [
          {arriveTangent=9.1628, interpMode=RCIM_Cubic, leaveTangent=9.1628, tangentMode=RCTM_User, time=0, value=0}
          {arriveTangent=-1.3283, interpMode=RCIM_Cubic, leaveTangent=-1.3283, tangentMode=RCTM_User, time=0.15, value=0.5}
          {tangentMode=RCTM_User, time=1, value=0}
         ]}}})
     B=&RandomRangeFloat(Minimum=1, Maximum=1.5)
    )
    Scale RGB=&Multiply_VectorByFloat(Vector=<1 1 1>, Float=&FNTimeOfDayScalar(Night Scale=0.5, Morning Scale=0.3, Evening Scale=0.4))
    ScaleA=true
    ScaleRGB=true
   )
   &ScaleMeshSize(Scale Factor=&MakeVector(
     X=&FloatFromCurve(FloatCurve=DataInterface() {propertyValues {Curve {keys [<at 0 0.8>,<at 0.5 1.2>]}}})
     Y=1
     Z=&FloatFromCurve(FloatCurve=DataInterface() {propertyValues {Curve {keys [<at 0 0.5>,<at 0.5 1>]}}})
    ))
   &ScaleMeshSize(Scale Factor=&VectorFromFloat(Value=&MakeCustomFloatFromBool(Bool=$User.bIsFrontEndPreview, True Float=0.5, False Float=0.9))) {module=ScaleMeshSize001}
  ]
  rendererClasses [&NiagaraMeshRendererProperties]
  renderers [{rendererClass=&NiagaraMeshRendererProperties, rendererIndex=0}]
 }
 glow001 {
  emitterUpdateScript [
   &EmitterState(Inactive Response=::NewEnumerator1, Life Cycle Mode=::NewEnumerator1, Loop Duration=0.375, UseLoopDelay=false)
   &SpawnBurst_Instantaneous(
    Spawn Probability=&MakeCustomFloatFromBool(Bool=$User.bIsFrontEndPreview, True Float=1, False Float=1)
    Use Spawn Probability=true
    UseLoopCountLimit=false
   )
  ]
  particleSpawnScript [
   &InitializeParticle(
    Color=<rgba 0.1 0.1 0.1 2>
    Lifetime=0.5
    Position Offset=<0 0 5>
    Sprite Size Mode=::NewEnumerator3
    Uniform Sprite Size=88
    UsePositionOffset=true
    Write Material Random=true
   )
   &CameraOffset(Camera Offset Amount=&Add_Float(
     A=4
     B=&FN_PickFloatDependOnPlatform(
      TargetPlatform=DataInterface() {propertyValues {Platforms {qualityLevelMask=-1}}}
      TargetPlatformActive=1
      TargetPlatformNoActive=&FN_FrontEndFloatModifier(Gameplay=5, Front End=2)
     )
    ))
  ]
  particleUpdateScript [
   &ParticleState()
   &ScaleSpriteSize(
    Uniform Curve Scale=&MakeCustomFloatFromBool(Bool=$User.bIsFrontEndPreview, True Float=0.7, False Float=1)
    Uniform Curve Sprite Scale=DataInterface() {propertyValues {Curve {keys [<at 0 1>, {tangentMode=RCTM_User, time=1, value=0.2}]}}}
   )
   /SetVariables(Particles.Size002:Vector2f=&Multiply_Vector2D(
     A=$Particles.SpriteSize
     B=&Vector2DFromFloat(Value=&FN_FrontEndFloatModifier(
       Gameplay=&FN_PickFloatDependOnPlatform(TargetPlatformActive=0.9, TargetPlatformNoActive=0.7)
       Front End=0.8
      ))
    )) {isSetParameters=true, module=SetVariables_#4}
   &ScaleColor(
    Scale Alpha=&FloatFromCurve(
     FloatCurve=DataInterface() {propertyValues {Curve {keys [
         {arriveTangent=3.589, interpMode=RCIM_Cubic, leaveTangent=3.589, tangentMode=RCTM_User, time=0, value=0.4}
         {arriveTangent=4.0483, interpMode=RCIM_Cubic, leaveTangent=-2.1547, tangentMode=RCTM_Break, time=0.1492, value=1}
         {arriveTangent=-0.8398, interpMode=RCIM_Cubic, leaveTangent=-0.8398, tangentMode=RCTM_User, time=0.3947, value=0.5753}
         {arriveTangent=-0.1619, interpMode=RCIM_Cubic, leaveTangent=-0.1619, tangentMode=RCTM_User, time=0.9841, value=0.3859}
        ]}}}
     Scale Curve=&Multiply_Float(
      A=&FNTimeOfDayScalar(Night Scale=0.7, Morning Scale=1, Evening Scale=1)
      B=&MakeCustomFloatFromBool(Bool=$User.bIsFrontEndPreview, True Float=0.5, False Float=1)
     )
    )
    Scale RGB=&MakeVectorFromLinearColorRGB(Color=&LerpLinearColors(StartColor=<rgba 0.0034 0.018 0.5469 1>, EndColor=<rgba 0.1778 0.2865 0.0307 1>, LerpFactor=$User.VariantSwitch))
    ScaleA=true
    ScaleRGB=true
   )
   /SetVariables(Particles.color002:LinearColor=&MakeLinearColorFromVectorAndFloat(
     "Vector (RGB)"=&MakeVectorFromLinearColorRGB(Color=&LerpLinearColors(StartColor=<rgba 3.8759 0.6783 6 1>, EndColor=<rgba 1.8502 3 0.2188 1>, LerpFactor=$User.VariantSwitch))
     "Float (Alpha)"=&FloatFromCurve(
      FloatCurve=DataInterface() {propertyValues {Curve {keys [
          {arriveTangent=-0.8058, interpMode=RCIM_Cubic, leaveTangent=-0.8058, tangentMode=RCTM_User, time=0, value=0.1718}
          {arriveTangent=-0.4971, interpMode=RCIM_Cubic, leaveTangent=-0.4971, tangentMode=RCTM_User, time=0.5, value=0}
         ]}}}
      Scale Curve=&Multiply_Float(
       A=&FNTimeOfDayScalar(Night Scale=0.6, Morning Scale=1, Evening Scale=1)
       B=&FN_FrontEndFloatModifier(Gameplay=1, Front End=0.4)
      )
     )
    )) {isSetParameters=true, module=SetVariables_#5}
   &FN_Light_Attributes(
    Color=&LerpLinearColors(StartColor=<rgba 0.0071 0.3473 1.5 1>, EndColor=<rgba 0.9595 1.5 0.1895 1>, LerpFactor=$User.VariantSwitch)
    Create Color Attribute=true
    Create Exponent Attribute=true
    Create Position Attribute=true
    Create Radius Attribute=true
    Exponent=2
    Radius=&FloatFromCurve(
     FloatCurve=DataInterface() {propertyValues {Curve {keys [<at 0 0.4>,<at 0.2 1.1>,<at 1 1.4>]}}}
     Scale Curve=&MakeCustomFloatFromBool(Bool=$User.bIsFrontEndPreview, True Float=0, False Float=100)
    )
    Scale Color=&FloatFromCurve(
     FloatCurve=DataInterface() {propertyValues {Curve {keys [
         {arriveTangent=4.1988, interpMode=RCIM_Cubic, leaveTangent=4.1988, tangentMode=RCTM_User, time=0, value=0}
         {arriveTangent=-0.0105, interpMode=RCIM_Cubic, leaveTangent=-0.0105, tangentMode=RCTM_User, time=0.15, value=1}
         {arriveTangent=-0.9284, interpMode=RCIM_Cubic, leaveTangent=-0.9284, tangentMode=RCTM_User, time=0.5, value=0}
        ]}}}
     Scale Curve=&Multiply_Float(A=&FNTimeOfDayScalar(Night Scale=0.5, Morning Scale=1, Evening Scale=1), B=20)
    )
   )
  ]
  rendererClasses [&NiagaraSpriteRendererProperties]
  renderers [{rendererClass=&NiagaraSpriteRendererProperties}]
 }
 glow002 {
  emitterUpdateScript [
   &EmitterState(Inactive Response=::NewEnumerator1, Life Cycle Mode=::NewEnumerator1, Loop Duration=0.375, UseLoopDelay=false)
   &SpawnBurst_Instantaneous(
    Spawn Probability=&MakeCustomFloatFromBool(Bool=$User.bIsFrontEndPreview, True Float=1, False Float=1)
    Use Spawn Probability=true
    UseLoopCountLimit=false
   )
  ]
  particleSpawnScript [
   &InitializeParticle(
    Color=<rgba 0.1 0.1 0.1 2>
    Lifetime=0.5
    Sprite Size Mode=::NewEnumerator3
    Uniform Sprite Size=5
    UsePositionOffset=false
    Write Material Random=true
   )
   &CameraOffset(Camera Offset Amount=&Add_Float(
     A=4
     B=&FN_PickFloatDependOnPlatform(
      TargetPlatform=DataInterface() {propertyValues {Platforms {qualityLevelMask=-1}}}
      TargetPlatformActive=1
      TargetPlatformNoActive=&FN_FrontEndFloatModifier(Gameplay=5, Front End=2)
     )
    ))
  ]
  particleUpdateScript [
   &ParticleState()
   &ScaleSpriteSize(
    Uniform Curve Scale=&MakeCustomFloatFromBool(Bool=$User.bIsFrontEndPreview, True Float=0.7, False Float=1)
    Uniform Curve Sprite Scale=DataInterface() {propertyValues {Curve {keys [<at 0 1>, {tangentMode=RCTM_User, time=1, value=0.2}]}}}
   )
   /SetVariables(Particles.Size002:Vector2f=&Multiply_Vector2D(
     A=$Particles.SpriteSize
     B=&Vector2DFromFloat(Value=&FN_FrontEndFloatModifier(
       Gameplay=&FN_PickFloatDependOnPlatform(TargetPlatformActive=0.9, TargetPlatformNoActive=0.7)
       Front End=0.8
      ))
    )) {isSetParameters=true, module=SetVariables_#4}
   &ScaleColor(
    Scale Alpha=&FloatFromCurve(
     FloatCurve=DataInterface() {propertyValues {Curve {keys [
         {arriveTangent=3.589, interpMode=RCIM_Cubic, leaveTangent=3.589, tangentMode=RCTM_User, time=0, value=0.4}
         {arriveTangent=4.0483, interpMode=RCIM_Cubic, leaveTangent=-2.1547, tangentMode=RCTM_Break, time=0.1492, value=1}
         {arriveTangent=-1.4669, interpMode=RCIM_Cubic, leaveTangent=-1.4669, tangentMode=RCTM_User, time=0.4, value=0.5}
         {arriveTangent=-0.1619, interpMode=RCIM_Cubic, leaveTangent=-0.1619, tangentMode=RCTM_User, time=0.9836, value=0}
        ]}}}
     Scale Curve=&Multiply_Float(
      A=&FNTimeOfDayScalar(Night Scale=0.7, Morning Scale=1, Evening Scale=1)
      B=&MakeCustomFloatFromBool(Bool=$User.bIsFrontEndPreview, True Float=0.5, False Float=1)
     )
    )
    Scale RGB=&MakeVectorFromLinearColorRGB(Color=&LerpLinearColors(StartColor=<rgba 0.0034 0.018 0.5469 1>, EndColor=<rgba 0.1778 0.2865 0.0307 1>, LerpFactor=$User.VariantSwitch))
    ScaleA=true
    ScaleRGB=true
   )
   /SetVariables(Particles.color002:LinearColor=&MakeLinearColorFromVectorAndFloat(
     "Vector (RGB)"=&MakeVectorFromLinearColorRGB(Color=&LerpLinearColors(StartColor=<rgba 0.2753 3.1115 6 1>, EndColor=<rgba 1.8502 3 0.2188 1>, LerpFactor=$User.VariantSwitch))
     "Float (Alpha)"=&FloatFromCurve(
      FloatCurve=DataInterface() {propertyValues {Curve {keys [
          {arriveTangent=-2.3913, interpMode=RCIM_Cubic, leaveTangent=-2.3913, tangentMode=RCTM_User, time=0, value=1}
          {arriveTangent=-0.4971, interpMode=RCIM_Cubic, leaveTangent=-0.4971, tangentMode=RCTM_User, time=0.5, value=0}
         ]}}}
      Scale Curve=&Multiply_Float(
       A=&FNTimeOfDayScalar(Night Scale=0.6, Morning Scale=1, Evening Scale=1)
       B=&FN_FrontEndFloatModifier(Gameplay=1, Front End=0.4)
      )
     )
    )) {isSetParameters=true, module=SetVariables_#5}
   &FN_Light_Attributes(
    Color=&LerpLinearColors(StartColor=<rgba 0.0071 0.3473 1.5 1>, EndColor=<rgba 0.9595 1.5 0.1895 1>, LerpFactor=$User.VariantSwitch)
    Create Color Attribute=true
    Create Exponent Attribute=true
    Create Position Attribute=true
    Create Radius Attribute=true
    Exponent=2
    Radius=&FloatFromCurve(
     FloatCurve=DataInterface() {propertyValues {Curve {keys [<at 0 0.4>,<at 0.2 1.1>,<at 1 1.4>]}}}
     Scale Curve=&MakeCustomFloatFromBool(Bool=$User.bIsFrontEndPreview, True Float=0, False Float=100)
    )
    Scale Color=&FloatFromCurve(
     FloatCurve=DataInterface() {propertyValues {Curve {keys [
         {arriveTangent=4.1988, interpMode=RCIM_Cubic, leaveTangent=4.1988, tangentMode=RCTM_User, time=0, value=0}
         {arriveTangent=-0.0105, interpMode=RCIM_Cubic, leaveTangent=-0.0105, tangentMode=RCTM_User, time=0.15, value=1}
         {arriveTangent=-0.9284, interpMode=RCIM_Cubic, leaveTangent=-0.9284, tangentMode=RCTM_User, time=0.5, value=0}
        ]}}}
     Scale Curve=&Multiply_Float(A=&FNTimeOfDayScalar(Night Scale=0.5, Morning Scale=1, Evening Scale=1), B=20)
    )
   )
  ]
  rendererClasses [&NiagaraSpriteRendererProperties]
  renderers [
   {rendererClass=&NiagaraSpriteRendererProperties}
   {rendererClass=&NiagaraSpriteRendererProperties, rendererIndex=1}
  ]
 }
}
summary {
 systemName=NS_Emote_RadioPaca_Laser_Side
 emitters [
  {emitterName=Rays, rendererClasses [&NiagaraMeshRendererProperties]}
  {emitterName=glow002, rendererClasses [&NiagaraSpriteRendererProperties]}
  {emitterName=glow001, rendererClasses [&NiagaraSpriteRendererProperties]}
  {emitterName=Thin_Rays, rendererClasses [&NiagaraMeshRendererProperties]}
  {bEnabled=false, emitterName=Particle, rendererClasses [&NiagaraSpriteRendererProperties]}
  {bEnabled=false, emitterName=Rays001, rendererClasses [&NiagaraMeshRendererProperties]}
  {emitterName=Particle001, rendererClasses [&NiagaraSpriteRendererProperties]}
 ]
 userVariables [
  User.BeamColorScale {type=NiagaraFloat, defaultValue=0.05, description=""}
  User.bIsFrontEnd {type=NiagaraBool}
  User.bIsFrontEndPreview {type=NiagaraBool}
  User.Color {type=LinearColor, description="", defaultValue {struct=LinearColor, value=<rgba 0.4817 0.5 0.4734 1>}}
  User.Color LUT Custom U Position {type=NiagaraFloat, defaultValue=0, description=""}
  User.FeetRestingPositions {
   type=NiagaraDataInterfaceArrayFloat3
   description=""
   defaultValue {
    struct=NiagaraExt_VariableValue_DataInterface
    value {
     dataInterface=&self:NiagaraDataInterfaceArrayFloat3_2
     dataInterfaceClass=&NiagaraDataInterfaceArrayFloat3
    }
   }
  }
  User.FlashesColorArray {
   type=NiagaraDataInterfaceArrayColor
   description=""
   defaultValue {
    struct=NiagaraExt_VariableValue_DataInterface
    value {
     dataInterface=&self:NiagaraDataInterfaceArrayColor_2
     dataInterfaceClass=&NiagaraDataInterfaceArrayColor
    }
   }
  }
  User.Jam? {type=NiagaraBool, defaultValue=true}
  User.LaserLengthMult {type=NiagaraFloat, defaultValue=1, description=""}
  User.LasersPositionArray {
   type=NiagaraDataInterfaceArrayFloat3
   description=""
   defaultValue {
    struct=NiagaraExt_VariableValue_DataInterface
    value {
     dataInterface=&self:NiagaraDataInterfaceArrayFloat3_3
     dataInterfaceClass=&NiagaraDataInterfaceArrayFloat3
    }
   }
  }
  User.LaserState {type=NiagaraFloat, defaultValue=0, description=""}
  User.Local Alpha Preview {type=NiagaraFloat, defaultValue=0.75, description=""}
  User.Other Alpha Preview {type=NiagaraFloat, defaultValue=0.25, description=""}
  User.SongBPM {type=NiagaraFloat, defaultValue=135, description=""}
  User.SpotlightColorArray {
   type=NiagaraDataInterfaceArrayColor
   description=""
   defaultValue {
    struct=NiagaraExt_VariableValue_DataInterface
    value {
     dataInterface=&self:NiagaraDataInterfaceArrayColor_0
     dataInterfaceClass=&NiagaraDataInterfaceArrayColor
    }
   }
  }
  User.SpotlightsOrientationArray {
   type=NiagaraDataInterfaceArrayFloat3
   description=""
   defaultValue {
    struct=NiagaraExt_VariableValue_DataInterface
    value {
     dataInterface=&self:NiagaraDataInterfaceArrayFloat3_1
     dataInterfaceClass=&NiagaraDataInterfaceArrayFloat3
    }
   }
  }
  User.SpotlightsPositionArray {
   type=NiagaraDataInterfaceArrayFloat3
   description=""
   defaultValue {
    struct=NiagaraExt_VariableValue_DataInterface
    value {
     dataInterface=&self:NiagaraDataInterfaceArrayFloat3_0
     dataInterfaceClass=&NiagaraDataInterfaceArrayFloat3
    }
   }
  }
  User.UseCustomColor {type=NiagaraFloat, defaultValue=1, description=""}
  User.UseMaterialColorOverride {type=NiagaraFloat, defaultValue=0, description=""}
  User.VariantSwitch {type=NiagaraFloat, defaultValue=0, description=""}
 ]
}
