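# Niagara system for the main lights of the RadioPaca emote: a colored spotlight with ambient bokeh and dust.
# SpotLightMain bursts a looping particle feeding mesh, light, component and sprite renderers for spotlight, flare and floor glow.
# Its color steps through User.SpotlightColorArray by System.Bar, with alpha pulsed by System.Oscillation and loop count, plus a flickering core color.
# Rainbow_bokeh and Rainbow_dust spawn sprites colored from that array, with bokeh scaled by time of day.
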NiagaraSystem /BRCosmetics/Animation/Game/MainPlayer/Emotes/RadioPaca/FX/NS_Emote_RadioPaca_MainLights
dependencies {
 usedDataInterfaces [&NiagaraDataInterfaceArrayFloat, &NiagaraDataInterfaceCurve]
 usedDynamicInputs [
  &MaskFloat
  &Divide_Float
  &Add_Integer
  /Niagara/DynamicInputs/Divide/Divide_Int
  &ModuloInt
 ]
 usedModules [
  &SystemState
  &self:SystemSpawnScript.SystemScriptSource.SystemScriptGraph.NiagaraNodeAssignment_0.SetVariables_#1
  &self:SystemSpawnScript.SystemScriptSource.SystemScriptGraph.NiagaraNodeAssignment_1.SetVariables_#2
  &self:SystemSpawnScript.SystemScriptSource.SystemScriptGraph.NiagaraNodeAssignment_2.SetVariables_#3
 ]
}
emitters {
 Rainbow_bokeh {
  emitterUpdateScript [
   &EmitterState()
   &SpawnRate(SpawnRate=4, Use Spawn Probability=false)
  ]
  particleSpawnScript [
   &InitializeParticle(
    Color=&SelectLinearColorFromArray(
     Array Sampling Mode=::NewEnumerator0
     Direct Array Index=?
     Direct Array Mode=::NewEnumerator1
     Color Selection Array=$User.SpotlightColorArray
     Random Seed=0
    )
    Lifetime Max=3
    Lifetime Min=0.3
    Lifetime Mode=::NewEnumerator1
    Sprite Rotation Angle=&UniformAOrBFloat(A=0, B=0.125, Evaluation Type=::NewEnumerator0, Recalculate Random Each Loop=true, Override Seed=false)
    Sprite Rotation Mode=::NewEnumerator1
    Sprite Size Mode=::NewEnumerator1
    Uniform Sprite Size Max=44
    Uniform Sprite Size Min=10
    UsePositionOffset=true
    Write Material Random=true
   )
   &ShapeLocation(
    MaskSpawnGroup=false
    Non Uniform Scale=<0.7 0.7 0.8>
    Offset=<0 0 85>
    Offset Mode=::NewEnumerator0
    Override Seed=false
    Shape Origin=<0 0 0>
    UseOwnerScale=true
   )
   &CameraOffset(Camera Offset Amount=-100)
  ]
  particleUpdateScript [
   &ParticleState()
   &ScaleSpriteSize(Uniform Curve Sprite Scale=DataInterface() {propertyValues {Curve {
       defaultValue=0.6687
       keys [
        {interpMode=RCIM_Cubic, tangentMode=RCTM_User, time=0, value=0}
        {arriveTangent=-0.4264, interpMode=RCIM_Cubic, leaveTangent=-0.4264, tangentMode=RCTM_User, time=0.5, value=1}
       ]
      }}}) {enabled=false}
   &ScaleColor(
    Scale Alpha=&FloatFromCurve(
     FloatCurve=DataInterface() {propertyValues {Curve {keys [
         {arriveTangent=9.8157, interpMode=RCIM_Cubic, leaveTangent=9.8157, tangentMode=RCTM_User, time=0, value=0}
         {arriveTangent=-0.0381, interpMode=RCIM_Cubic, leaveTangent=-0.0381, tangentMode=RCTM_User, time=0.1675, value=0.6}
         {interpMode=RCIM_Cubic, tangentMode=RCTM_User, time=0.3, value=0.35}
         {interpMode=RCIM_Cubic, tangentMode=RCTM_User, time=0.5, value=0.5}
         {interpMode=RCIM_Cubic, tangentMode=RCTM_Break, time=1, value=0}
        ]}}}
     Scale Curve=&FNTimeOfDayScalar(Night Scale=0.2, Morning Scale=0.35, Day Scale=0.4, Evening Scale=0.35)
    )
    ScaleA=true
    ScaleRGB=true
   )
   &Drag(Drag=2.5, Use Linear Drag=true, Use Rotational Drag=false) {enabled=false}
   &GravityForce(Gravity=&RandomRangeVector(Minimum=<-5 -5 -5>, Maximum=<5 -5 5>))
   &SolveForcesAndVelocity(Clamp Velocity=false, Limit Acceleration=false)
   /SetVariables(Particles.New Position:NiagaraPosition=&Add_Vector(
     A=Linked() {linkedVariable {name=Particles.Position, type=NiagaraPosition}}
     B=Linked() {linkedVariable {name=Engine.Owner.Position, type=NiagaraPosition}}
    )) {isSetParameters=true}
  ]
  rendererClasses [&NiagaraSpriteRendererProperties]
  renderers [{rendererClass=&NiagaraSpriteRendererProperties}]
 }
 Rainbow_dust {
  emitterUpdateScript [
   &EmitterState(
    Inactive Response=::NewEnumerator1
    Life Cycle Mode=::NewEnumerator1
    Loop Behavior=::NewEnumerator1
    Loop Duration Mode=::NewEnumerator1
    UseLoopDelay=false
   )
   &SpawnRate(SpawnRate=5, Use Spawn Probability=true)
  ]
  particleSpawnScript [
   &InitializeParticle(
    Color=&SelectLinearColorFromArray(
     Array Sampling Mode=::NewEnumerator0
     Direct Array Index=?
     Direct Array Mode=::NewEnumerator1
     Color Selection Array=$User.SpotlightColorArray
     Random Seed=0
    )
    Lifetime=&RandomRangeFloat(Minimum=1, Maximum=1.5)
    Sprite Rotation Mode=::NewEnumerator2
    Sprite Size Mode=::NewEnumerator1
    Uniform Sprite Size Max=35
    Uniform Sprite Size Min=25
    UsePositionOffset=false
    Write Material Random=true
   )
   &AddVelocity(Velocity=<0 0 0>)
   &SystemLocation(MaskSpawnGroup=false)
   &ShapeLocation(MaskSpawnGroup=false, Override Seed=false, Sphere Radius=75, UseOwnerScale=true)
  ]
  particleUpdateScript [
   &ParticleState()
   &SubUVAnimation(
    End Frame=15
    SubUV Animation Mode:ENiagara_SubUVLookupMode=::NewEnumerator1
    SubUV Curve=DataInterface() {propertyValues {Curve {keys [{tangentMode=RCTM_User, time=0, value=0}, {tangentMode=RCTM_User, time=1.0102, value=0.6518}]}}}
   )
   &Drag(
    Drag=&Multiply_Float(
     A=&RandomRangeFloat(Minimum=0.5, Maximum=1)
     B=&FloatFromCurve(FloatCurve=DataInterface() {propertyValues {Curve {keys [
          {interpMode=RCIM_Cubic, leaveTangent=0.248, tangentMode=RCTM_Break, time=0, value=0}
          {interpMode=RCIM_Cubic, tangentMode=RCTM_Break, time=1, value=1}
         ]}}})
    )
    Use Linear Drag=true
    Use Rotational Drag=false
   )
   &SolveForcesAndVelocity(Clamp Velocity=false, Limit Acceleration=false)
   &ScaleColor(
    Scale Alpha=&FloatFromCurve(
     FloatCurve=DataInterface() {propertyValues {Curve {keys [
         {arriveTangent=5, interpMode=RCIM_Cubic, time=0, value=0}
         {arriveTangent=1.5181, interpMode=RCIM_Cubic, leaveTangent=1.5181, time=0.2, value=0.7}
         {
          arriveTangent=0.1817
          arriveTangentWeight=0.1842
          interpMode=RCIM_Cubic
          leaveTangent=0.1817
          leaveTangentWeight=0.7369
          tangentMode=RCTM_User
          time=0.5037
          value=0.7647
         }
         {arriveTangent=-1.5409, interpMode=RCIM_Cubic, leaveTangent=-1.5409, time=0.8, value=0.7376}
         <at 1 0>
        ]}}}
     Scale Curve=&RandomRangeFloat(Minimum=0.02, Maximum=0.07)
    )
    ScaleA=true
    ScaleRGB=true
   )
   &ScaleSpriteSize(
    Uniform Curve Scale=10
    Uniform Curve Sprite Scale=DataInterface() {propertyValues {Curve {keys [
        {interpMode=RCIM_Cubic, leaveTangent=0.6647, tangentMode=RCTM_Break, time=0, value=1}
        {arriveTangent=-0.0319, interpMode=RCIM_Cubic, tangentMode=RCTM_Break, time=1.0059, value=1.1871}
       ]}}}
   )
  ]
  rendererClasses [&NiagaraSpriteRendererProperties]
  renderers [{rendererClass=&NiagaraSpriteRendererProperties}]
 }
 SpotLightMain {
  emitterUpdateScript [
   &EmitterState(
    Inactive Response=::NewEnumerator1
    Life Cycle Mode=::NewEnumerator1
    Loop Behavior=::NewEnumerator1
    Loop Duration Mode=::NewEnumerator1
    UseLoopDelay=false
   )
   &SpawnBurst_Instantaneous(Use Spawn Probability=false, UseLoopCountLimit=false)
  ]
  particleSpawnScript [
   &InitializeParticle(
    Mesh Scale=<0.8 0.8 1.2>
    Mesh Scale Mode=::NewEnumerator4
    Sprite Size Mode=::NewEnumerator3
    Uniform Sprite Size=333
    UsePositionOffset=false
    Write Material Random=true
   )
   &CameraOffset()
   /SetVariables(
    Particles.FloorGlowFacingVector:Vector3f=<0.01 0.01 1>
    Particles.FloorGlowPosition:NiagaraPosition=&AddVectorToPosition(Position=/Niagara/DynamicInputs/Helpers/SimulationPosition(), Vector=<0 0 1>)
    Particles.FloorGlowSize:Vector2f=<500 500>
    Particles.LightAttenuationScale=275
    Particles.LightExponent=2
    Particles.LightPosition:NiagaraPosition=&AddVectorToPosition(Position=&SimulationPosition(), Vector=<0 0 200>)
    Particles.LightRadius=230
    Particles.SpotLightPosition:NiagaraPosition=&AddVectorToPosition(Position=&SimulationPosition(), Vector=<0 0 280>)
   ) {isSetParameters=true}
  ]
  particleUpdateScript [
   &ParticleState(
    Kill Particles When Lifetime Has Elapsed=false
    Let Infinitely Lived Particles Die When Emitter Deactivates=true
    Loop Particles Lifetime=true
   )
   &Color(
    Color=&SelectLinearColorFromArray(
     Array Sampling Mode=::NewEnumerator1
     Direct Array Index=$System.Bar
     Direct Array Mode=::NewEnumerator1
     Color Selection Array=$User.SpotlightColorArray
     Random Seed=?
    )
    Scale Alpha=&Multiply_Float(
     A=$System.Oscillation
     B=&SelectFloatFromArray(
      Direct Array Index=$System.LoopCount
      Direct Array Mode=::NewEnumerator1
      Float Selection Array=DataInterface() {propertyValues {FloatData [1,0.7,1,0.7,1,0.7]}}
     )
    )
   )
   /SetVariables(
    Particles.CoreSpotlightColor:LinearColor=&MultiplyLinearColor(
     Linear Color=<rgba 0.2 0.1687 0.16 0.5>
     Scale Factor=&MakeFloat4FromVectorAndFloat("Vector (RGB)"=<1 1 1>, "Float (Alpha)"=$System.Flicker)
    )
    Particles.FlareColor=&MultiplyLinearColor(Linear Color=$Particles.Color, Scale Factor=&MakeVector4_FromFloat(Float=0.4))
    Particles.FloorGlowColor:LinearColor=&MultiplyLinearColor(Linear Color=$Particles.Color, Scale Factor=&MakeVector4_FromFloat(Float=0.3))
    Particles.LightColor=&MultiplyLinearColor(Linear Color=$Particles.Color, Scale Factor=&MakeVector4_FromFloat(Float=8))
    Particles.LightComponentIntensity:NiagaraFloat=&Multiply_Float(A=1, B=500)
   ) {isSetParameters=true}
  ]
  rendererClasses [
   &NiagaraMeshRendererProperties
   &NiagaraLightRendererProperties
   &NiagaraComponentRendererProperties
   &NiagaraSpriteRendererProperties
  ]
  renderers [
   {rendererClass=&NiagaraMeshRendererProperties, rendererIndex=0}
   {rendererClass=&NiagaraMeshRendererProperties, rendererIndex=1}
   {rendererClass=&NiagaraLightRendererProperties, rendererIndex=2}
   {rendererClass=&NiagaraComponentRendererProperties, rendererIndex=3}
   {rendererClass=&NiagaraSpriteRendererProperties, rendererIndex=4}
  ]
 }
}
summary {
 systemName=NS_Emote_RadioPaca_MainLights
 emitters [
  {
   emitterName=SpotLightMain
   rendererClasses [
    &NiagaraMeshRendererProperties
    &NiagaraLightRendererProperties
    &NiagaraComponentRendererProperties
    &NiagaraSpriteRendererProperties
   ]
  }
  {emitterName=Rainbow_bokeh, rendererClasses [&NiagaraSpriteRendererProperties]}
  {emitterName=Rainbow_dust, rendererClasses [&NiagaraSpriteRendererProperties]}
 ]
 userVariables [
  User.BeamColorScale {type=NiagaraFloat, defaultValue=0.05, description=""}
  User.bIsFrontEnd {type=NiagaraBool}
  User.bIsFrontEndPreview {type=NiagaraBool}
  User.Color {type=LinearColor, description="", defaultValue {struct=LinearColor, value=<rgba 0.272 0.2823 0.2673 1>}}
  User.Color LUT Custom U Position {type=NiagaraFloat, defaultValue=0, description=""}
  User.FeetRestingPositions {
   type=NiagaraDataInterfaceArrayFloat3
   description=""
   defaultValue {
    struct=NiagaraExt_VariableValue_DataInterface
    value {
     dataInterface=&self:NiagaraDataInterfaceArrayFloat3_2
     dataInterfaceClass=&NiagaraDataInterfaceArrayFloat3
    }
   }
  }
  User.FlashesColorArray {
   type=NiagaraDataInterfaceArrayColor
   description=""
   defaultValue {
    struct=NiagaraExt_VariableValue_DataInterface
    value {
     dataInterface=&self:NiagaraDataInterfaceArrayColor_2
     dataInterfaceClass=&NiagaraDataInterfaceArrayColor
    }
   }
  }
  User.Jam? {type=NiagaraBool, defaultValue=true}
  User.LaserLengthMult {type=NiagaraFloat, defaultValue=1, description=""}
  User.LasersPositionArray {
   type=NiagaraDataInterfaceArrayFloat3
   description=""
   defaultValue {
    struct=NiagaraExt_VariableValue_DataInterface
    value {
     dataInterface=&self:NiagaraDataInterfaceArrayFloat3_3
     dataInterfaceClass=&NiagaraDataInterfaceArrayFloat3
    }
   }
  }
  User.LaserState {type=NiagaraFloat, defaultValue=0, description=""}
  User.Local Alpha Preview {type=NiagaraFloat, defaultValue=0.75, description=""}
  User.Other Alpha Preview {type=NiagaraFloat, defaultValue=0.25, description=""}
  User.SongBPM {type=NiagaraFloat, defaultValue=135, description=""}
  User.SpotlightColorArray {
   type=NiagaraDataInterfaceArrayColor
   description=""
   defaultValue {
    struct=NiagaraExt_VariableValue_DataInterface
    value {
     dataInterface=&self:NiagaraDataInterfaceArrayColor_0
     dataInterfaceClass=&NiagaraDataInterfaceArrayColor
    }
   }
  }
  User.SpotlightsOrientationArray {
   type=NiagaraDataInterfaceArrayFloat3
   description=""
   defaultValue {
    struct=NiagaraExt_VariableValue_DataInterface
    value {
     dataInterface=&self:NiagaraDataInterfaceArrayFloat3_1
     dataInterfaceClass=&NiagaraDataInterfaceArrayFloat3
    }
   }
  }
  User.SpotlightsPositionArray {
   type=NiagaraDataInterfaceArrayFloat3
   description=""
   defaultValue {
    struct=NiagaraExt_VariableValue_DataInterface
    value {
     dataInterface=&self:NiagaraDataInterfaceArrayFloat3_0
     dataInterfaceClass=&NiagaraDataInterfaceArrayFloat3
    }
   }
  }
  User.UseCustomColor {type=NiagaraFloat, defaultValue=1, description=""}
  User.UseMaterialColorOverride {type=NiagaraFloat, defaultValue=0, description=""}
 ]
}
